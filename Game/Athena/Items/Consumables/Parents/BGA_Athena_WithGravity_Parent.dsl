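; Parent Blueprint for Athena consumable gameplay actors that fly under projectile movement, bounce, and settle on surfaces.
; Hits within the slope limit stop simulation and can attach to the struck actor; steeper ones bounce.
; Support death, door opening or physics waking restarts the fall, unless nearby buildings provide new support where checked.
; The server drives replicated resume state and dormancy, backing stops with a client RPC and location snap against desync.
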(event Custom|HandleProjectileStopped
  (params (ImpactResult :type Hit%20Result%20Structure%20%28by%20ref%29))
  ;[1 1 1] React to code side projectile stop event
  (if (not (Variables|Default|GetbCalledStopSim))
    (CallFunction|StopSim ImpactResult) ; Call BP stop sim logic if haven't done so
  ))

(event Custom|SpawnFXSounds)

(event EventBeginPlay
  ;[1 1 1] On Begin Play Check For Resim
  (|CallParentFunction)
  (CallFunction|SpawnFXSounds)
  (Utilities|FlowControl|Sequence
    (:then_0
      ;[1 1 1] React to code side projectile stop event
      (Networking|SwitchHasAuthority
        (:Authority
          (EventDispatchers|BindEventtoOnProjectileStop (Variables|Default|GetProjectileMovement) (delegate Custom|HandleProjectileStopped)))))
    (:then_1
      (if (Variables|Default|GetResumeGravSimOnBegin)
        (CallFunction|RestartSimulation)
        (Components|Movement|SetUpdatedComponent (Variables|Default|GetProjectileMovement) (Variables|Default|GetStaticMesh))
        (Utilities|Array|Add (Variables|Default|GetActorsToNotAttachTo) self)
        (else
          (CallFunction|StopSim (Variables|Default|GetNullHit))
          (Development|PrintString "STOP" (Variables|Default|GetDebugText) (Variables|Default|GetDebugText)))))))

(event Custom|StopSimOnClients
  (params (HitResult :type Hit%20Result%20Structure))
  ;[1 1 1] If we trigger a hit as soon as resume simulation is set on the server, the server will stomp the resume simulation replication and the client won't receive it's stop sim call from the replicated ResumeSim bool. RPC to client here to make sure we're stopping sim on client as well in this condition. This was causing bugs with the client projectile moving while the parent had stopped simulating. This will early return in stop sim if the client has already called stop sim.
  (CallFunction|StopSim HitResult))

(event Custom|StopSim
  (params (Hit :type Hit%20Result%20Structure))
  (bind (_bblockinghit _binitialoverlap _time _distance _location _impactpoint _normal _impactnormal _physmat _hitactor _hitcomponent) (Collision|BreakHitResult Hit))
  ;[1 1 1] Handle Falling
  ;;[1 1 1] Stop Sim
  (if (not (or (Variables|Default|GetBlockStoppingSim) (Variables|Default|GetbCalledStopSim)))
    (Variables|Default|SetbCalledStopSim true)
    (Development|PrintString "Stopped" (Variables|Default|GetDebugText) (Variables|Default|GetDebugText))
    (Networking|SwitchHasAuthority
      (:Authority
        (Variables|Default|SetStopSimHitResult Hit))
      (:Remote
        _))
    (Game|Components|ProjectileMovement|StopSimulating (Variables|Default|GetProjectileMovement) Hit)
    (Networking|SwitchHasAuthority
      (:Authority
        (Variables|Default|SetbResumeSimulation false)
        (CallFunction|StopSimonClients Hit)
        (Networking|SetNetDormancy "DORM_DormantAll"))
      (:Remote
        _))
    (Networking|SwitchHasAuthority
      (:Authority
        (Variables|Default|SetStopSimServerLocation (Transformation|GetActorLocation self)))
      (:Remote
        _))
    (if (Variables|Default|GetDeactivatePawnAndVehicleCollisionOnStop)
      (Collision|SetCollisionResponsetoChannel (Variables|Default|GetStaticMesh) "ECC_Vehicle" "ECR_Ignore")
      (Collision|SetCollisionResponsetoChannel (Variables|Default|GetStaticMesh) "ECC_Pawn" "ECR_Ignore")
      (else
        _))
    (Utilities|IsValid _hitactor
      (:Is%20Valid
        (Variables|Default|SetGravImpact_Loc _impactpoint)
        (Networking|SwitchHasAuthority
          (:Authority
            (Variables|Default|SetGravHitNormal _normal))
          (:Remote
            _))
        (CallFunction|MulticastHandleBinding _hitactor _hitcomponent)
        (CallFunction|MulticastPlayHitFX Hit)
        (Networking|ForceNetUpdate)))))

(event Custom|Multicast_PlayHitFX
  (params (HitResult :type Hit%20Result%20Structure)))

(event Custom|Multicast_HandleBinding
  (params (Actor :type Actor%20Object%20Reference) (HitComp :type Primitive%20Component%20Object%20Reference))
  ;[1 1 1] Handle Falling
  ;;[1 1 1] Binds Event for when Building piece dies
  (Utilities|IsWater Actor
    (:true
      (CallFunction|RestartSimulation)
      (Variables|/Script/Engine.MovementComponent|SetVelocity "0, 0,700.000000" (Variables|Default|GetProjectileMovement)) ; Needs this to ensure it registers collision with the water.
    )
    (:false
      (Utilities|Array|Add (Variables|Default|GetActorsToNotAttachTo) Actor)
      (Utilities|IsA(soft) Actor "/Game/Athena/PlayerPawn_Athena.PlayerPawn_Athena_C"
        (:True
          (CallFunction|RestartSimulation))
        (:False
          (bind _asmilitary_base_door_01 (Utilities|Casting|CastToMilitaryBase_Door_01 Actor)
            (:then
              (Utilities|FlowControl|Sequence
                (:then_0
                  (if (Utilities|Equal(Object) (Variables|/Game/Creative/Sets/MilitaryBase/BuildingPieces/MilitaryBase_Door_01.MilitaryBase_Door_01_C|GetDoorStaticMesh1 _asmilitary_base_door_01) HitComp)
                    (Default|BindEventtoOnDoorOpened _asmilitary_base_door_01 (EventDispatchers|CreateEvent :!event Custom|SlidingDoorOpened))))
                (:then_1
                  _)))
            (:CastFailed
              _))
          (bind _asbuilding_actor (Utilities|Casting|CastToBuildingActor Actor)
            (:then
              (bind _returnvalue (FortPhysicsSimulationLibrary|CanActorSimulatePhysics _asbuilding_actor))
              (if _returnvalue
                (bind _returnvalue_2 (BuildingActor|PhysicsSimulationSupport|GetPhysicsObjectComponent _asbuilding_actor))
                (PhysicsObject|BindEventtoOnAwakeChanged _returnvalue_2 (delegate Custom|OnUnderneathPhysicsObjectAwakeChanged)))
              (EventDispatchers|BindEventtoOnDied _asbuilding_actor (delegate Custom|BaseDied)))
            (:CastFailed
              (Utilities|IsValid Actor
                (:Is%20Valid
                  (Game|BindEventtoOnDestroyed Actor (delegate Custom|BaseDestroyed)))
                (:Is%20Not%20Valid
                  (CallFunction|RestartSimulation)))))
          (CallFunction|AttachtoBindedActor HitComp))))))

(event Collision|EventHit
  (params (MyComp :type Primitive%20Component%20Object%20Reference) (Other :type Actor%20Object%20Reference) (OtherComp :type Primitive%20Component%20Object%20Reference) (bSelfMoved :type Boolean) (HitLocation :type Vector :default "0, 0, 0") (HitNormal :type Vector :default "0, 0, 0") (NormalImpulse :type Vector :default "0, 0, 0") (Hit :type Hit%20Result%20Structure%20%28by%20ref%29))
  (bind (_bblockinghit _binitialoverlap _time _distance _location _impactpoint _normal) (Collision|BreakHitResult Hit))
  (bind (_asbga_athena_with_gravity_parent _bsuccess) (Utilities|Casting|CastToBGA_Athena_WithGravity_Parent Other))
  (bind (_asfort_projectile_base _bsuccess_2) (Utilities|Casting|CastToFortProjectileBase Other))
  ;[1 1 1] Handle Falling
  ;;[1 1 1] When this actor hits a flat enough surface, it stops simulating movement.
  (if (Variables|Default|GetbResumeSimulation)
    (if (or (and _bsuccess (Variables|Default|GetDontAttachToOthersOfThisClass)) _bsuccess_2)
      _
      (elif (Utilities|Array|ContainsItem (Variables|Default|GetActorsToNotAttachTo) Other)
        _
        (else
          ;;;[1 1 1] Handle PlayerPawn_Athena Hit
          (Utilities|IsA(soft) Other "/Game/Athena/PlayerPawn_Athena.PlayerPawn_Athena_C"
            (:True
              (Collision|SetCollisionResponsetoChannel (Variables|Default|GetStaticMesh) "ECC_Pawn" "ECR_Ignore") ; Stops player and shield bubble from sticking together in the air if player is on shield bubble and destroys ground beneath them.
              (Utilities|FlowControl|Delay 0.1)
              (CallFunction|RestartSimulation))
            (:False
              ;;;[1 1 1] Bounce off AI pawn
              (Utilities|Casting|CastToFortAIPawn Other
                (:then
                  _)
                (:CastFailed
                  ;;;[1 1 1] Bounce off vehicles
                  (if (Variables|Default|GetbBounceOffVehicles)
                    (Utilities|Casting|CastToFortVehicleInterface Other
                      (:then
                        _)
                      (:CastFailed
                        ;;;[1 1 1] Bounce Off Tires
                        (if (not (Variables|Default|GetBounceOffTires))
                          _
                          (elif (not (Utilities|String|Contains (Utilities|String|ToString(Name) (Transformation|GetAttachSocketName OtherComp)) "tire"))
                            _
                            (else
                              (CallFunction|BounceBGAAwayfromLocation (Transformation|GetActorLocation Other)))))
                        ;;;[1 1 1] General Hit handling
                        (if (< (Math|Trig|Acos(Degrees) (Math|Vector|DotProduct HitNormal "0, 0,1.000000")) (Variables|Default|GetGravMaxSlope))
                          (if (Variables|Default|GetCheckForBounce)
                            (bind _hitsameclass (CallFunction|CheckforSameClasstoBounce Other))
                            (if _hitsameclass
                              _
                              (elif (and (Variables|Default|GetForceBounce) (> (Variables|Default|GetForcedBounceMaxCount) (Variables|Default|GetForcedBounceCurrentCount)))
                                (CallFunction|ForceBounceBGA)))
                            (else
                              _))
                          (CallFunction|StopSim Hit)
                          (Development|PrintString "STOP" (Variables|Default|GetDebugText) (Variables|Default|GetDebugText))
                          (else
                            (CallFunction|BounceBGA _normal)))))))))))))))

(event Custom|SlidingDoorOpened
  ;[1 1 1] Handle Falling
  ;;[1 1 1] Base destroyed. Start simulating again.
  (CallFunction|AttachedWasDestroyed) ; Event that can be overridden in children
  (if (not (Variables|Default|GetShouldNotReattach))
    (bind (_outactors _returnvalue) (Collision|SphereOverlapActors (Transformation|GetActorLocation) 16.0 (Utilities|Array|MakeArray "ObjectTypeQuery1") "/Script/FortniteGame.BuildingActor" (Variables|Default|GetActorsToNotAttachTo)))
    (if _returnvalue
      (bind _output_get (Variables|Default|SetGravFoundBuildingOnDied _outactors))
      (if (and (.bSuccess (Utilities|Casting|CastToBGA_Athena_WithGravity_Parent (Utilities|Array|Get(acopy) _output_get))) (Variables|Default|GetDontAttachToOthersOfThisClass))
        (CallFunction|RestartSimulation)
        (else
          (bind _asbuilding_actor (Utilities|Casting|CastToBuildingActor (Utilities|Array|Get(acopy) _output_get))
            (:then
              (EventDispatchers|BindEventtoOnDied _asbuilding_actor (EventDispatchers|CreateEvent :!event Custom|BaseDied)))
            (:CastFailed
              (Game|BindEventtoOnDestroyed _asbuilding_actor (EventDispatchers|CreateEvent :!event Custom|BaseDestroyed))))))
      (else
        (CallFunction|RestartSimulation)))))

(event Custom|BaseDied
  (params (DamagedActor :type Actor%20Object%20Reference) (Damage :type Float%20%28single-precision%29) (InstigatedBy :type Controller%20Object%20Reference) (DamageCauser :type Actor%20Object%20Reference) (HitLocation :type Vector) (FHitComponent :type Primitive%20Component%20Object%20Reference) (BoneName :type Name) (Momentum :type Vector))
  ;[1 1 1] Handle Falling
  ;;[1 1 1] Base destroyed. Start simulating again.
  (CallFunction|AttachedWasDestroyed) ; Event that can be overridden in children
  (if (not (Variables|Default|GetShouldNotReattach))
    (bind (_outactors _returnvalue) (Collision|SphereOverlapActors (Transformation|GetActorLocation) 16.0 (Utilities|Array|MakeArray "ObjectTypeQuery1") "/Script/FortniteGame.BuildingActor" (Variables|Default|GetActorsToNotAttachTo)))
    (if _returnvalue
      (bind _output_get (Variables|Default|SetGravFoundBuildingOnDied _outactors))
      (if (and (.bSuccess (Utilities|Casting|CastToBGA_Athena_WithGravity_Parent (Utilities|Array|Get(acopy) _output_get))) (Variables|Default|GetDontAttachToOthersOfThisClass))
        (CallFunction|RestartSimulation)
        (else
          (bind _asbuilding_actor (Utilities|Casting|CastToBuildingActor (Utilities|Array|Get(acopy) _output_get))
            (:then
              (EventDispatchers|BindEventtoOnDied _asbuilding_actor (EventDispatchers|CreateEvent :!event Custom|BaseDied)))
            (:CastFailed
              (Game|BindEventtoOnDestroyed _asbuilding_actor (EventDispatchers|CreateEvent :!event Custom|BaseDestroyed))))))
      (else
        (CallFunction|RestartSimulation)))))

(event Custom|BaseDestroyed
  (params (DestroyedActor :type Actor%20Object%20Reference))
  ;[1 1 1] Handle Falling
  ;;[1 1 1] Base destroyed. Start simulating again.
  (CallFunction|AttachedWasDestroyed) ; Event that can be overridden in children
  (if (not (Variables|Default|GetShouldNotReattach))
    (bind (_outactors _returnvalue) (Collision|SphereOverlapActors (Transformation|GetActorLocation) 16.0 (Utilities|Array|MakeArray "ObjectTypeQuery1") "/Script/FortniteGame.BuildingActor" (Variables|Default|GetActorsToNotAttachTo)))
    (if _returnvalue
      (bind _output_get (Variables|Default|SetGravFoundBuildingOnDied _outactors))
      (if (and (.bSuccess (Utilities|Casting|CastToBGA_Athena_WithGravity_Parent (Utilities|Array|Get(acopy) _output_get))) (Variables|Default|GetDontAttachToOthersOfThisClass))
        (CallFunction|RestartSimulation)
        (else
          (bind _asbuilding_actor (Utilities|Casting|CastToBuildingActor (Utilities|Array|Get(acopy) _output_get))
            (:then
              (EventDispatchers|BindEventtoOnDied _asbuilding_actor (EventDispatchers|CreateEvent :!event Custom|BaseDied)))
            (:CastFailed
              (Game|BindEventtoOnDestroyed _asbuilding_actor (EventDispatchers|CreateEvent :!event Custom|BaseDestroyed))))))
      (else
        (CallFunction|RestartSimulation)))))

(event Custom|AttachedWasDestroyed)

(event Custom|OnUnderneathPhysicsObjectAwakeChanged
  (params (SimulatingComponent :type Primitive%20Component%20Object%20Reference) (bIsAwake :type Boolean))
  ;[1 1 1] Handle Falling
  ;;[1 1 1] Base destroyed. Start simulating again.
  (if bIsAwake
    (bind _asbuilding_actor (Utilities|Casting|CastToBuildingActor (Components|GetOwner SimulatingComponent)))
    (bind _returnvalue (BuildingActor|PhysicsSimulationSupport|GetPhysicsObjectComponent _asbuilding_actor))
    (PhysicsObject|UnbindEventfromOnAwakeChanged _returnvalue (delegate Custom|OnUnderneathPhysicsObjectAwakeChanged))
    (CallFunction|BounceBGAAwayfromLocation (Transformation|GetActorLocation _asbuilding_actor))
    (CallFunction|RestartSimulation)))

(event StickyProjectile|EventDetachStickyProjectile
  ;[1 1 1] In case we want to explicitly detach from an object.
  (CallFunction|BaseDestroyed 0))

(event Custom|Impulse
  (params (DelayBeforeImpulse :type Float%20%28double-precision%29) (DelayBeforeCollision :type Float%20%28double-precision%29) (ImpulseOrigin :type Vector) (ImpulseAmount :type Vector) (SetCollisionAfterImpulse :type Boolean))
  ;[1 1 1] Impulse Actor
  (Utilities|FlowControl|Delay DelayBeforeImpulse)
  (Variables|/Script/Engine.MovementComponent|SetVelocity (* (Math|Vector|Normalize (- (Transformation|GetActorLocation) ImpulseOrigin)) ImpulseAmount) (Variables|Default|GetProjectileMovement))
  (Utilities|FlowControl|Delay DelayBeforeCollision)
  (Variables|Default|SetRepCollision SetCollisionAfterImpulse))

(fn AttachToBindedActor
  (params (AttachComp :type Primitive%20Component%20Object%20Reference))
  (bind _returnvalue (Components|GetOwner AttachComp))
  (Utilities|IsValid AttachComp
    (:Is%20Valid
      (if (Variables|Default|GetShouldAttach)
        (Utilities|Casting|CastToFortVehicleInterface _returnvalue
          (:then
            _)
          (:CastFailed
            (bind _asbuilding_actor (Utilities|Casting|CastToBuildingActor _returnvalue)
              (:then
                ;[1 1 1] Don't attach to physics objects, this causes very sstrange collision behavior: FORT-440815
                (bind _returnvalue_2 (FortPhysicsSimulationLibrary|CanActorSimulatePhysics _asbuilding_actor))
                (if _returnvalue_2
                  _
                  (else
                    _)))
              (:CastFailed
                (Utilities|Casting|CastToBuildingWall _returnvalue)))))
        (Transformation|AttachActorToComponent AttachComp "None" "KeepWorld" "KeepWorld" "KeepWorld")
        (if (not (Variables|Default|GetAllowReattachmentToActors))
          (Collision|IgnoreActorwhenMoving (Variables|Default|GetStaticMesh) _returnvalue true))
        (CallFunction|OnAttach _returnvalue)))))

(fn BounceBGA
  (params (BounceNormal :type Vector :default "0.000000,0.000000,1.000000"))
  ;[1 1 1] Let the BGA Bounce and adjust its velocity.
  (Variables|/Script/Engine.MovementComponent|SetVelocity (Math|Vector|vector+vector (Math|Vector|MirrorVectorbyNormal (* (Variables|/Script/Engine.MovementComponent|GetVelocity (Variables|Default|GetProjectileMovement)) (Variables|Default|GetBounceMult)) (Math|Vector|SelectVector (Variables|GetBounceNormal) "0.000000,0.000000,1.000000" (Variables|Default|GetbBounceOnHitNormal))) (Math|Vector|MakeVector 0.0 0.0 (* (Math|Vector|VectorLength (Variables|/Script/Engine.MovementComponent|GetVelocity (Variables|Default|GetProjectileMovement))) (Variables|Default|GetBounceExtraZ)))) (Variables|Default|GetProjectileMovement))
  (CallFunction|MulticastPlayHitFX)
  (Networking|ForceNetUpdate)
  (if (Variables|Default|GetCountNonForceBounces)
    (Math|Integer|IncrementInt (Variables|Default|GetForcedBounceCurrentCount))))

(fn BounceBGAAwayFromLocation
  (params (LocationToBounceAwayFrom :type Vector))
  (Variables|/Script/Engine.MovementComponent|SetVelocity (Utilities|Operators|Multiply (Math|Vector|Normalize (- (Transformation|GetActorLocation self) LocationToBounceAwayFrom)) (Math|Float|SelectFloat (/ (Variables|Default|GetBounceAwayFromLocation_MinSpeed) (Math|Float|Max(Float) (Variables|Default|GetBounceMult) 0.01)) (Math|Vector|VectorLength (Variables|/Script/Engine.MovementComponent|GetVelocity (Variables|Default|GetProjectileMovement))) (< (Math|Vector|VectorLength (Variables|/Script/Engine.MovementComponent|GetVelocity (Variables|Default|GetProjectileMovement))) (Variables|Default|GetBounceAwayFromLocation_MinSpeed))) (Variables|Default|GetBounceMult)) (Variables|Default|GetProjectileMovement))
  (CallFunction|MulticastPlayHitFX)
  (Networking|ForceNetUpdate))

(fn CheckForSameClassToBounce
  (params (Hit :type Actor%20Object%20Reference))
  (bind _returnvalue (Utilities|GetClass Hit))
  (bind _self self)
  (bind _returnvalue_2 (Utilities|GetClass _self))
  (bind _returnvalue_3 (Utilities|Equal(Class) _returnvalue _returnvalue_2))
  (bind _additionalbounceobject (Variables|Default|GetAdditionalBounceObject))
  (bind _returnvalue_4 (Utilities|Equal(Object) _additionalbounceobject Hit))
  ;[1 1 1] Check if the hit was on an actor of the same class or on is the additional ignore actor.
  (if _returnvalue_3
    (CallFunction|BounceBGA "0.000000,0.000000,1.000000")
    (return _returnvalue_3)
    (else
      (Utilities|IsValid (Variables|Default|GetAdditionalBounceObject)
        (:Is%20Valid
          (if _returnvalue_4
            (CallFunction|BounceBGA "0.000000,0.000000,1.000000")
            (return _returnvalue_4)
            (else
              (return _returnvalue_4))))
        (:Is%20Not%20Valid
          (return _returnvalue_3))))))

(fn ForceBounceBGA
  ;[1 1 1] Let the CBGA Bounce and adjust its velocity.
  (Variables|/Script/Engine.MovementComponent|SetVelocity (Math|Vector|vector+vector (Math|Vector|MirrorVectorbyNormal (* (Variables|/Script/Engine.MovementComponent|GetVelocity (Variables|Default|GetProjectileMovement)) (Variables|Default|GetForcedBounceMult)) "0,0.000000,1.000000") (Math|Vector|MakeVector 0.0 0.0 (* (Math|Vector|VectorLength (Variables|/Script/Engine.MovementComponent|GetVelocity (Variables|Default|GetProjectileMovement))) (Variables|Default|GetForcedBounceExtraZ)))) (Variables|Default|GetProjectileMovement))
  (CallFunction|MulticastPlayHitFX)
  (Networking|ForceNetUpdate)
  (Math|Integer|IncrementInt (Variables|Default|GetForcedBounceCurrentCount)))

(fn HandleResumeSimulationChanged)

(fn Init
  (params (GravHitNormal :type Vector))
  (Variables|Default|SetGravHitNormal GravHitNormal))

(fn OnAttach
  (params (AttachedActor :type Actor%20Object%20Reference)))

(fn OnRep_bResumeSimulation
  (CallFunction|HandleResumeSimulationChanged)
  (if (Variables|Default|GetbResumeSimulation)
    (Variables|Default|SetbCalledStopSim false)
    (Components|Movement|SetUpdatedComponent (Variables|Default|GetProjectileMovement) (Variables|Default|GetStaticMesh))
    (Development|PrintString "Did Actually Resume" (Variables|Default|GetDebugText) (Variables|Default|GetDebugText))
    (if (Variables|Default|GetDeactivatePawnAndVehicleCollisionOnStop)
      (Collision|SetCollisionResponsetoChannel (Variables|Default|GetStaticMesh) "ECC_Vehicle" "ECR_Block")
      (Collision|SetCollisionResponsetoChannel (Variables|Default|GetStaticMesh) "ECC_Pawn" "ECR_Overlap")
      (else
        _))
    (else
      (Game|Components|ProjectileMovement|StopSimulating (Variables|Default|GetProjectileMovement) (Variables|Default|GetStopSimHitResult)))))

(fn OnRep_RepCollision
  (Collision|SetActorEnableCollision (Variables|Default|GetRepCollision)))

(fn OnRep_StopLocation)

(fn OnRep_StopSimServerLocation
  ;[1 1 1] Kinda hacky fix for syncing client and server locations on stop sim
  (if (not (Variables|Default|GetbResumeSimulation))
    (Transformation|SetActorLocation self (Variables|Default|GetStopSimServerLocation) false true)))

(fn RestartSimulation
  (Networking|SwitchHasAuthority
    (:Authority
      (Variables|Default|SetbResumeSimulation true)
      (Networking|SetNetDormancy "DORM_Awake")
      (Networking|FlushNetDormancy))))

(fn ConstructionScript)
