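; Parent projectile for thrown Athena consumables, which flies, bounces and becomes a world pickup.
; Bounces alternate between explosion with FX and a one-time, server-side pickup spawn that follows moving platforms; stopping explodes and kills it.
; First water entry splashes, then after a delay spawns and tosses the pickup, sending a quest stat.
; Helpers broadcast creation, choose pickup owner by game mode and ping the replicated pickup; homing events retune turn speed.
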(event EventBeginPlay
  ;[0 1 0.149] Begin Play
  (Variables|VFX|SetFX%20Linger%20Time (Math|Random|RandomFloatinRange 1.0 1.0))
  (CallFunction|PlayThrowSound)
  (Networking|SwitchHasAuthority
    (:Authority
      (bind _returnvalue (Utilities|Time|SetTimerbyEvent (delegate Custom|Timeout) (Variables|Default|GetProjectile%20Max%20Lifetime)))
      (Variables|Default|SetLifespan%20Timer _returnvalue)
      (CallFunction|BroadcastPickupProjectileCreated))))

(event Custom|Timeout)

(event ProjectileMovement|EventOnHomingPassedTarget
  ;[1 1 1] Homing
  (Projectile|SetHomingTurnSpeedMin 0 0 3.6e+02 :!class /Script/FortniteGame.FortProjectileMovementInterface)
  (Projectile|SetHomingTurnSpeedMax 0 0 7.2e+02 :!class /Script/FortniteGame.FortProjectileMovementInterface)
  (Projectile|SetHomingTimeUntilMaxTurnSpeed 0 0 0.5 :!class /Script/FortniteGame.FortProjectileMovementInterface))

(event ProjectileMovement|EventOnReachedLockTargetDistanceThreshold
  ;[1 1 1] Homing
  (Projectile|SetHomingTurnSpeedMin 0 0 3.6e+02 :!class /Script/FortniteGame.FortProjectileMovementInterface)
  (Projectile|SetHomingTurnSpeedMax 0 0 7.2e+02 :!class /Script/FortniteGame.FortProjectileMovementInterface)
  (Projectile|SetHomingTimeUntilMaxTurnSpeed 0 0 0.5 :!class /Script/FortniteGame.FortProjectileMovementInterface))

(event EventOnStop
  (params (Hit :type Hit%20Result%20Structure%20%28by%20ref%29))
  (bind (_bblockinghit _binitialoverlap _time _distance _location _impactpoint) (Collision|BreakHitResult Hit))
  ;[1 1 1] Event on Stop play Hit Fx
  (Variables|VFX|SetImpact%20Location _impactpoint)
  (Utilities|FlowControl|Sequence
    (:then_0
      (if (not (Variables|Default|GetbHasExploded?))
        (Variables|Default|SetbHasExploded? true)
        (Delivery|DoExplosion) ; Explosion
        (Delivery|Kill)))
    (:then_1
      (Development|SetHiddeninGame (Variables|Default|GetStaticMesh) true false)
      (Networking|SwitchHasAuthority
        (:Authority
          (Utilities|Time|ClearandInvalidateTimerbyHandle (Variables|Default|GetLifespan%20Timer)))
        (:Remote
          (Utilities|FlowControl|Delay (Variables|VFX|GetFX%20Linger%20Time))
          (Actor|DestroyActor))))))

(event OnEnterWaterEvent(WaterInteractionComponent)
  (params (WaterBody :type Fort%20Water%20Body%20Actor%20Object%20Reference) (WaterInteractionComponent :type Fort%20Water%20Interaction%20Component%20Object%20Reference) (bIsFirstBody :type Boolean))
  (bind _pickup (Variables|Default|GetPickup))
  (bind _returnvalue (Projectile|GetFortProjectileMovementComponent _pickup))
  (bind _itemdefinition (Variables|Default|GetItemDefinition))
  ;[1 1 1] Spawn the item on Enter Water
  (if bIsFirstBody
    (CallFunction|DeactivateTrailFXSystem)
    (CallFunction|WaterImpactItemFX)
    (Utilities|FlowControl|Delay (Variables|Default|GetWaterSpawnDelay))
    (if (not (Variables|Default|GetbHasExploded?))
      (CallFunction|SpawnItem true)
      (Pickup|TossPickupinDirection _pickup "0.000000,0.000000,1.000000" 0 false true) ; Make sure it's simulating
      (Utilities|IsValid _returnvalue
        (:Is%20Valid
          (Variables|/Script/Engine.ProjectileMovementComponent|SetMaxSpeed (Variables|Default|GetMaxSpeedInWater) _returnvalue) ; Reset the max speed to something sane 
        )
        (:Is%20Not%20Valid
          _))
      ;[1 1 1] Send a quest update with tag Athena.Quests.Throw.Consumables.InWater
      (bind _asfort_player_pawn_athena (Utilities|Casting|CastToFortPlayerPawnAthena (Actor|GetOwner)))
      (Utilities|IsValid _itemdefinition
        (:Is%20Valid
          (Quest|SendComplexCustomStatEvent (Quests|GetQuestManager (Fortnite|GetFortPlayerControllerfromActor _asfort_player_pawn_athena) "Invalid") 0 "(GameplayTags=((TagName=\"Athena.Quests.Throw.Consumables.InWater\")))" (GameplayTags|GetOwnedGameplayTags _itemdefinition :!class /Script/GameplayTags.GameplayTagAssetInterface) 1))))))

(event EventOnBounce
  (params (Hit :type Hit%20Result%20Structure%20%28by%20ref%29))
  (bind (_bblockinghit _binitialoverlap _time _distance _location _impactpoint _normal _impactnormal _physmat _hitactor) (Collision|BreakHitResult Hit))
  (bind (_bblockinghit_2 _binitialoverlap_2 _time_2 _distance_2 _location_2 _impactpoint_2 _normal_2 _impactnormal_2) (Collision|BreakHitResult Hit))
  (bind _staticmesh (Variables|Default|GetStaticMesh))
  ;[1 1 1] BounceEvent
  (Variables|Default|SetMovingPlatform (Actor|GetComponentbyClass _hitactor "/Script/FortniteGame.FortMovingPlatformComponent"))
  (Utilities|SendStatThrowConsumables
    (:Out
      (Utilities|FlowControl|FlipFlop
        (:A
          (Delivery|DoExplosion) ; Explosion
          (CallFunction|DeactivateTrailFXSystem)
          ;;[1 1 1] UnHideTracerMesh
          (Rendering|SetVisibility _staticmesh true false)
          (Development|SetHiddeninGame _staticmesh false false)
          ;;[0 0.3344 1] Play Bounce off ground/wall Fx
          (Effects|Components|ParticleSystem|SpawnFXAtLocation (Variables|Default|GetBounce%20Off%20FX) _impactpoint_2 (Math|Conversions|RotationFromXVector _impactnormal_2)))
        (:B
          (Utilities|FlowControl|DoOnce false
            (:Completed
              (bind _movingplatform (Variables|Default|GetMovingPlatform)
                (:then
                  (Transformation|AttachActorToComponent (Variables|Transformation|GetRootComponent (Components|GetOwner _movingplatform)) "None" "KeepWorld" "KeepWorld" "KeepWorld" false) ; After bouncing off a moving platform, move relative to it.
                )
                (:else
                  _))
              (Networking|SwitchHasAuthority
                (:Authority
                  (bind _optional_owner_pc (CallFunction|GetSpawnedPickupOptionalOwnerPC))
                  (bind _returnvalue (Inventory|SpawnPickupInWorld (Variables|Default|GetItemDefinition) 1 (+ (Transformation|GetActorLocation) (Math|Vector|SelectVector (Variables|Default|GetSpawnPositionOffsetWater) (Variables|Default|GetSpawnPositionOffset) (stale Custom|SpawnItem bHitWater))) (Transformation|GetActorForwardVector) 0 false (Variables|Default|GetShouldUseRandomRotationWhenSpawningPickUp) false -1 "Tossed" "TossedByPlayer" _optional_owner_pc))
                  (CallFunction|OnPickupSpawned _returnvalue)
                  (CallFunction|SpawnedItemFx)
                  (bind _returnvalue_2 (PlayerInfo|GetOwnerPlayerController))
                  (Class|FortPickup|SetDropper _returnvalue _returnvalue_2)
                  (bind _movingplatform_2 (Variables|Default|GetMovingPlatform)
                    (:then
                      (Class|FortMovingPlatformComponent|NotifyPickupStopped _movingplatform_2 _returnvalue Hit) ; Attach the spawned pickup to the moving patform.
                    )
                    (:else)))))))))) ; Complex Custom\nAthena.Quests.Throw.Consumables
)

(event Custom|SpawnedItemFx
  ;[1 1 1] BounceEvent
  ;;[0 0.3381 1] Play Spawned Item FX
  (Effects|Components|ParticleSystem|SpawnFXAtLocation (Variables|Default|GetSpawnFX) (Transformation|GetActorLocation) (Math|Conversions|RotationFromXVector (.ImpactNormal (Collision|BreakHitResult (stale EventOnBounce Hit)))))
  (Audio|PlaySoundatLocation (Variables|Sound|GetSpawnSound) (Transformation|GetActorLocation))
  (Components|Movement|StopMovementImmediately (Variables|Movement|GetProjectileMovementComponent))
  (Development|SetHiddeninGame (Variables|Default|GetStaticMesh) true false)
  (Delivery|Kill))

(event Custom|SpawnItem
  (params (bHitWater :type Boolean))
  ;[1 1 1] BounceEvent
  (Utilities|FlowControl|DoOnce false
    (:Completed
      (bind _movingplatform (Variables|Default|GetMovingPlatform)
        (:then
          (Transformation|AttachActorToComponent (Variables|Transformation|GetRootComponent (Components|GetOwner _movingplatform)) "None" "KeepWorld" "KeepWorld" "KeepWorld" false) ; After bouncing off a moving platform, move relative to it.
        )
        (:else
          _))
      (Networking|SwitchHasAuthority
        (:Authority
          (bind _optional_owner_pc (CallFunction|GetSpawnedPickupOptionalOwnerPC))
          (bind _returnvalue (Inventory|SpawnPickupInWorld (Variables|Default|GetItemDefinition) 1 (+ (Transformation|GetActorLocation) (Math|Vector|SelectVector (Variables|Default|GetSpawnPositionOffsetWater) (Variables|Default|GetSpawnPositionOffset) bHitWater)) (Transformation|GetActorForwardVector) 0 false (Variables|Default|GetShouldUseRandomRotationWhenSpawningPickUp) false -1 "Tossed" "TossedByPlayer" _optional_owner_pc))
          (CallFunction|OnPickupSpawned _returnvalue)
          (CallFunction|SpawnedItemFx)
          (bind _returnvalue_2 (PlayerInfo|GetOwnerPlayerController))
          (Class|FortPickup|SetDropper _returnvalue _returnvalue_2)
          (bind _movingplatform_2 (Variables|Default|GetMovingPlatform)
            (:then
              (Class|FortMovingPlatformComponent|NotifyPickupStopped _movingplatform_2 _returnvalue (stale EventOnBounce Hit)) ; Attach the spawned pickup to the moving patform.
            )
            (:else)))))))

(event Custom|WaterImpactItemFX
  ;[1 1 1] BounceEvent
  ;;[0 0.3381 1] Play Water Impact FX
  (Effects|Components|ParticleSystem|SpawnFXAtLocation (Variables|Default|GetSpawnFX_Water) (Transformation|GetActorLocation) (Math|Conversions|RotationFromXVector (.ImpactNormal (Collision|BreakHitResult (stale EventOnBounce Hit)))))
  (Audio|PlaySoundatLocation (Variables|Sound|GetSplash%20Sound%20) (Transformation|GetActorLocation))
  (Development|SetHiddeninGame (Variables|Default|GetStaticMesh) true false))

(event EventOnExploded
  (params (HitActors :type Array%20of%20Actor%20Object%20References) (HitResults :type Array%20of%20Hit%20Result%20Structures))
  (if (Variables|/Script/Engine.ProjectileMovementComponent|GetbShouldBounce (Variables|Movement|GetProjectileMovementComponent))
    (if (not (< (Math|Trig|Acos(Degrees) (Math|Vector|DotProduct (.Normal (Collision|BreakHitResult (stale EventOnBounce Hit))) (Math|Vector|VectorUp))) 3e+01))
      ;[1 1 1] Bounce Off Walls
      (Variables|/Script/Engine.ProjectileMovementComponent|SetbRotationFollowsVelocity false (Variables|Movement|GetProjectileMovementComponent))
      (CallFunction|DeactivateTrailFXSystem))))

(event Custom|Mark
  (params (NewParam :type Actor%20Object%20Reference))
  ;[1 1 1] CUSTOM EVENT -> Ping the Item after it spawns
  (bind _asfort_player_pawn_athena (Utilities|Casting|CastToFortPlayerPawnAthena (Actor|GetOwner)))
  (bind _asfort_player_controller_athena (Utilities|Casting|CastToFortPlayerControllerAthena (Pawn|GetController _asfort_player_pawn_athena)))
  (Fortnite|RegisterSoundIndicatorPlacementContext _asfort_player_controller_athena NewParam))

(event Custom|SetItemDefinition
  (params (In%20Item%20Definition :type Fort%20World%20Item%20Definition%20Object%20Reference))
  ;[0.18 1 0.6503] Set Item Definition
  (Variables|Default|SetItemDefinition In%20Item%20Definition)
  (CallFunction|BroadcastPickupProjectileCreated))

(event Custom|Broadcast%20Pickup%20Projectile%20Created
  ;[0.343 0.7225 1] Once we have a valid item definition, broadcast through the player controller that a pickup projectile has been created.
  (Utilities|IsValid (Variables|Default|GetItemDefinition)
    (:Is%20Valid
      (Utilities|FlowControl|DoOnce false
        (:Completed
          (bind _returnvalue (PlayerInfo|GetOwnerPlayerController))
          (bind _asfort_player_controller_athena (Utilities|Casting|CastToFortPlayerControllerAthena _returnvalue))
          (Class|FortPlayerControllerAthena|BroadcastonPickupProjectileCreated _asfort_player_controller_athena (Variables|Default|GetItemDefinition)))))))

(fn DeactivateTrailFXSystem
  (bind _fx_trail (Variables|VFX|GetFX_Trail))
  (bind _niagara_trail (Variables|Default|GetNiagara_Trail))
  (Utilities|IsValid _fx_trail
    (:Is%20Valid
      (Components|Activation|Deactivate _fx_trail))
    (:Is%20Not%20Valid
      _))
  (Utilities|IsValid _niagara_trail
    (:Is%20Valid
      (Components|Activation|Deactivate _niagara_trail))))

(fn GetSpawnedPickupOptionalOwnerPC
  ;[0.969 1 0.3804] Check if it is a creative Playlist if so pass in the Optional Owner
  (bind _asfort_game_state_athena (Utilities|Casting|CastToFortGameStateAthena (Game|GetGameState))
    (:then
      (if (not (and (Class|FortGameStateAthena|IsCreativeModePlaylist _asfort_game_state_athena) (FortGlobals|IsinPhoenixZone)))
        (bind _asfort_game_mode (Utilities|Casting|CastToFortGameMode (Game|GetGameMode) :!label _label_2)
          (:then
            (if (Class|FortGameMode|ShouldInstanceInventoryDrops _asfort_game_mode)
              ;[1 0.4635 0.4635] Return Owner Player Controller
              (bind _returnvalue (PlayerInfo|GetOwnerPlayerController :!label _label))
              (return _returnvalue)
              (else
                _)))
          (:CastFailed
            _))
        ;[1 0.4635 0.4635] Return nullptr
        (return 0))
      (goto _label))
    (:CastFailed
      _))
  (goto _label_2))

(fn OnPickupSpawned
  (params (Spawned%20Pickup :type Fort%20Pickup%20Object%20Reference))
  (Variables|Default|SetPickup Spawned%20Pickup)
  (Utilities|IsValid Spawned%20Pickup
    (:Is%20Valid
      (bind _returnvalue (PlayerInfo|GetOwnerPlayerController))
      (Utilities|IsValid _returnvalue
        (:Is%20Valid
          (Class|FortPlayerController|BroadcastonPickupCreated _returnvalue Spawned%20Pickup))
        (:Is%20Not%20Valid
          _)))
    (:Is%20Not%20Valid
      _))
  (return))

(fn OnRep_ImpactLocation
  (CallFunction|DeactivateTrailFXSystem))

(fn OnRep_Pickup
  (Networking|CanExecuteCosmeticEvents
    (:True
      (CallFunction|Mark (Variables|Default|GetPickup)))))

(fn PlayThrowSound
  (bind _throwsoundattachedtoplayer (Variables|Sound|GetThrowSoundAttachedToPlayer)
    (:then
      (bind _asfort_pawn (Utilities|Casting|CastToFortPawn (Game|GetInstigator)))
      (Audio|SpawnSoundAttached _throwsoundattachedtoplayer (Variables|/Script/Engine.Character|GetMesh _asfort_pawn) (Components|SkinnedMesh|GetSocketBoneName (Variables|/Script/Engine.Character|GetMesh _asfort_pawn) "hand_r")))
    (:else)))

(fn ConstructionScript)
