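; Parent gameplay ability for throwing Athena consumables, with overridable hooks for child abilities.
; Activation applies a windup effect and cook montage, then throws instantly, on input release, or on primary press for secondary fire.
; A locally controlled indicator previews the predicted spline with bounces and flags blocked deployment areas.
; The throw montage's trigger spawns the projectile unless the ability is blocked; cleanup removes windup effects and any use-spawned indicator.
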(event Ability|EventActivateAbility
  ;[0.06083 1 0] ActivateAbility
  (if (ScalableFloat|GetValueasBool (Variables|ParentDefaults|GetbIsEnabled))
    (Rendering|SetActorHiddenInGame (Variables|Trajectory|GetTrajectoryIndicator) false)
    (CallFunction|AddValidCamera)
    (ProjectileLifecycle|PrethrowSetup) ; Ability is instanced per actor, so we need to reset variables\nevery time the ability is activated.
    (bind _montage (CallFunction|GetCookMontage))
    (Ability|Tasks|PlayMontageAndWait "None" _montage)
    (GameplayCues|AddCookGC)
    (if (or (Variables|ParentDefaults|GetbThrowInstantly) (ParentDefaults|IsDoingThrowDown))
      ;;[0.2354 0.1465 1] Instant throwing
      (Variables|AbilityState|SetbIsThrowing true)
      (CallFunction|ThrowProjectile)
      (else
        (TrajectoryInterface|SpawnTrajectoryIndicator false)
        (if (Variables|ParentDefaults|GetbIsSecondaryFire)
          ;;[0.5836 0 1] Secondary fire - throw when primary fire is pressed
          (if (and (Variables|ParentDefaults|GetbAllowPrimaryFireOnTouchSecondaryRelease) (Inputs|IsUsingInputType "Touch" (Fortnite|GetFortPlayerControllerfromActor (Default|GetOwningPlayer))))
            (Ability|Tasks|WaitGameplayEvent "(TagName=\"Event.TouchConsumableThrown\")" 0 true
              (:then
                (Networking|CanExecuteCosmeticEvents
                  (:False
                    _)))
              (:EventReceived
                _))
            (Ability|Tasks|WaitNetSync "OnlyServerWait"
              (:OnSync
                (Variables|AbilityState|SetbIsThrowing true)
                (CallFunction|StartThrowProjectile)))
            (else
              (Ability|Tasks|WaitGameplayEvent "(TagName=\"Event.Primary.Pressed\")" 0 true
                (:then
                  _)
                (:EventReceived
                  _))))
          (else
            ;;[0.269 1 0.501] Throw the projectile if input is released while still in throw windup.
            (Ability|Tasks|WaitInputRelease false
              (:OnRelease
                (Variables|AbilityState|SetAbilityKeyPressed false)
                (if (Variables|AbilityState|GetInThrowWindup)
                  (CallFunction|StartThrowProjectile))))))))
    (else
      (Ability|EndAbility))))

(event Custom|StartThrowProjectile
  ;[1 1 1] Throw the actual projectile, can be overrriden
  (CallFunction|ThrowProjectile))

(event Ability|EventOnEndAbility
  (params (bWasCancelled :type Boolean))
  (|CallParentFunction bWasCancelled)
  (Utilities|FlowControl|Sequence
    (:then_0
      ;[1 0.522 0.08636] Failsafe to clean up the trajectory when the ability ends.
      (ProjectileLifecycle|PostThrowCleanup))
    (:then_1
      ;[1 1 0] Cancel the throw montage async task
      (bind _throwmontageasynctask (Variables|Default|GetThrowMontageAsyncTask)
        (:then
          (GameplayTasks|EndTask _throwmontageasynctask))
        (:else)))))

(event Ability|EventOnAbilityInputReleased
  ;[0.5836 0 1] Secondary Fire - Cancel ability on release.
  (if (Variables|ParentDefaults|GetbIsSecondaryFire)
    (Variables|AbilityState|SetAbilityKeyPressed false)
    (if (not (Variables|AbilityState|GetbIsThrowing))
      (Ability|EndAbility))))

(event TrajectoryInterface|EventInitTrajectoryVariables
  ;[0.1828 0.8499 1] Projectile Trajectory
  (if (not (Variables|Trajectory|GetProjectile%20Trajectory%20Data%20Initialized))
    (bind _output_get (Variables|Trajectory|SetProjectileGravityScaleToUse (LocationandRotation|GetProjectileInitialGravityScaletoUse)))
    (bind _trajectorydata (Class|FortKismetLibrary|GetProjectileTrajectoryData (Default|GetOwningPlayer) (Variables|ParentDefaults|GetThrownProjectileClass) (LocationandRotation|GetProjectileInitialSpeedtoUse) _output_get))
    (Variables|Trajectory|SetProjectileTrajectoryData _trajectorydata)
    (Variables|Trajectory|SetProjectile%20Trajectory%20Data%20Initialized true)))

(event Custom|ThrowMontageStarted)

(event Custom|ThrowProjectile
  ;[0.3521 1 0.1896] Throw Projectile - Trigger the throw montage and spawn the projectile when the notify is hit.
  ;;[1 1 0] If the ability became blocked, end it.
  ;; Do not start the throw montage.
  (if (Default|IsAbilityBlocked)
    (Ability|EndAbility)
    (else
      ;;[0.2299 0.2304 1] BR: ShouldBlockAbilities = true
      ;; StW: CanBeCanceled = false,  ShouldBlockAbilities = true
      (Ability|SetShouldBlockOtherAbilities true)
      (ProjectileLifecycle|PostThrowCleanup)
      (bind _montage (CallFunction|GetThrowMontage))
      ;;[0.2299 0.2304 1] Projectile is spawned when the montage reaches ability notify
      (bind _asynctaskproxy (Ability|Tasks|PlayMontageandWaitforTrigger "ThrowProjectile" _montage)
        (:then
          (Variables|Default|SetThrowMontageAsyncTask _asynctaskproxy)
          ;;[0.2299 0.2304 1] Can overide these functions in children
          (CallFunction|ThrowMontageStarted))
        (:Triggered
          ;;[1 1 0 1 14] If the ability became blocked, end it.
          ;; Do not spawn a projectile.
          (if (not (Default|IsAbilityBlocked))
            (ProjectileLifecycle|AttemptSpawnThrownProjectile)))
        (:Cancelled
          _)
        (:Completed
          (Ability|SetShouldBlockOtherAbilities false)))
      (Ability|EndAbility))))

(event TrajectoryInterface|EventUpdateTrajectory
  ;[0.1828 0.8499 1] Projectile Trajectory
  (Trajectory|UpdateTrajectorySpline))

(event TrajectoryInterface|EventCleanupTrajectoryIndicatorOnUnequip
  ;[0.1828 0.8499 1] Projectile Trajectory
  (Trajectory|CleanupTrajectory))

(fn AddValidCamera
  (bind _cameramode (Variables|ParentDefaults|GetCameraMode))
  ;[1 1 1] If Valid add a camera for the duration of the ability
  (if (Utilities|IsValidClass _cameramode)
    (Camera|PushCameraMode _cameramode)))

(fn CalculateProjectileTrajectory
  (params (Instigator :type Fort%20Pawn%20Object%20Reference) (InitialLocation :type Vector) (InitialVelocity :type Vector))
  (bind _projectiletrajectorydata (Variables|Trajectory|GetProjectileTrajectoryData))
  (bind (_speed _gravityz _capsuleradius _capsulehalfheight _bshouldbounce _bounciness _friction) (Utilities|Struct|BreakFortProjectileTrajectoryData _projectiletrajectorydata))
  (bind (_speed_2 _gravityz_2 _capsuleradius_2 _capsulehalfheight_2 _bshouldbounce_2 _bounciness_2 _friction_2) (Utilities|Struct|BreakFortProjectileTrajectoryData _projectiletrajectorydata))
  (bind _trajectorycollisionprofilename (Variables|Trajectory|GetTrajectoryCollisionProfileName))
  (Utilities|FlowControl|Sequence
    (:then_0
      ;[0.1354 1 0.1527] Initialize Projectile Ignored Actors
      (Utilities|IsValid Instigator
        (:Is%20Valid
          (LocalVariables|SetProjectileIgnoredActors (Class|FortPawn|GetProjectileIgnoredActors Instigator))
          (Utilities|Array|AppendArray (LocalVariables|GetProjectileIgnoredActors) (Variables|Trajectory|GetTrajectoryIgnoredWorldActors)))))
    (:then_1
      ;[0.3008 0.2867 1] Call Helper functions to calculate the trajectory points
      (if (Utilities|Name|NotEqual(Name) _trajectorycollisionprofilename "None")
        ;;[0.2107 1 0.704] Calculate with a profile name
        (bind (_outhitresult _outsplinepoints _outsplinetangents) (Projectile|CalculateProjectileTrajectorySplinewithHitProfile (Variables|GetInstigator) (Variables|GetInitialLocation) (Variables|GetInitialVelocity) _speed_2 _gravityz_2 _friction_2 _bounciness_2 (Variables|Trajectory|GetTrajectoryTimeStep) (Math|Float|Max(Float) _capsuleradius_2 _capsulehalfheight_2) _trajectorycollisionprofilename (LocalVariables|GetProjectileIgnoredActors) (select _bshouldbounce_2 (Variables|Trajectory|GetTrajectory%20Max%20Bounces) 1) (Variables|Trajectory|GetTrajectoryMaxSteps) (Variables|Trajectory|GetTrajectory%20Max%20Distance%20Between%20Spline%20Points) (Variables|Trajectory|GetTrajectory%20Initial%20Distance) 0.0 (Variables|Trajectory|GetReturnPhysicalMaterial)))
        (return _outhitresult _outsplinepoints _outsplinetangents)
        (else
          ;;[0.1083 1 0.0636] Calculate without a profile name.
          (bind (_outhitresult_2 _outsplinepoints_2 _outsplinetangents_2) (Projectile|CalculateProjectileTrajectorySplinewithHit (Variables|GetInstigator) (Variables|GetInitialLocation) (Variables|GetInitialVelocity) _speed _gravityz _friction _bounciness (Variables|Trajectory|GetTrajectoryTimeStep) (Math|Float|Max(Float) _capsuleradius _capsulehalfheight) "ECC_GameTraceChannel8" (LocalVariables|GetProjectileIgnoredActors) (select _bshouldbounce (Variables|Trajectory|GetTrajectory%20Max%20Bounces) 1) (Variables|Trajectory|GetTrajectoryMaxSteps) (Variables|Trajectory|GetTrajectory%20Max%20Distance%20Between%20Spline%20Points) (Variables|Trajectory|GetTrajectory%20Initial%20Distance) 0.0 (Variables|Trajectory|GetReturnPhysicalMaterial)))
          (return _outhitresult_2 _outsplinepoints_2 _outsplinetangents_2))))))

(fn CleanupTrajectory
  (bind _trajectoryindicator (Variables|Trajectory|GetTrajectoryIndicator)
    (:then
      (Class|BPProjectileTrajectory|SetShouldUpdatefromOwner _trajectoryindicator false 0)
      (Rendering|SetActorHiddenInGame _trajectoryindicator true)
      (Actor|DestroyActor _trajectoryindicator))
    (:else)))

(fn GetThrowMontage
  (bind _throwmontage1p (Variables|ParentDefaults|GetThrowMontage1P))
  (bind _asfort_player_controller (Utilities|Casting|CastToFortPlayerController (Pawn|GetController (Default|GetOwningPlayer)))
    (:then
      (if (Camera|IsUsingFirstPersonCamera _asfort_player_controller)
        (Utilities|IsValid _throwmontage1p
          (:Is%20Valid
            (return _throwmontage1p))
          (:Is%20Not%20Valid
            _))))
    (:CastFailed
      _))
  (return (Variables|ParentDefaults|GetThrowMontage)))

(fn GetCookMontage
  (bind _returnvalue (Default|GetOwningPlayer))
  (bind _cookmontage1p (Variables|ParentDefaults|GetCookMontage1P))
  (bind _asfort_player_controller (Utilities|Casting|CastToFortPlayerController (Pawn|GetController _returnvalue))
    (:then
      (if (Camera|IsUsingFirstPersonCamera _asfort_player_controller)
        (Utilities|IsValid _cookmontage1p
          (:Is%20Valid
            (return _cookmontage1p))
          (:Is%20Not%20Valid
            _))
        (else
          (if (Variables|ParentDefaults|GetbUse%20Crouch%20Specific%20Cook%20Montage)
            (if (Variables|/Script/Engine.Character|GetbIsCrouched _returnvalue)
              (return (Variables|ParentDefaults|GetCrouchCookMontage))
              (else
                _)))))
      (return (Variables|ParentDefaults|GetCookMontage)))
    (:CastFailed
      _)))

(fn GetProjectileInitialGravityScaleToUse
  (bind _thrownprojectileclass (Variables|ParentDefaults|GetThrownProjectileClass))
  (bind _initialgravityscaleoverride (Class|GetClassDefaults _thrownprojectileclass))
  (bind _getprojectileinitialgravityscalecontext (LocalVariables|GetGetProjectileInitialGravityScaleContext))
  (if (Utilities|IsValidClass _thrownprojectileclass)
    (if (Fortnite|IsCurveTableRowValid _initialgravityscaleoverride _getprojectileinitialgravityscalecontext)
      (bind _outxy (Fortnite|EvaluateCurveTableRow _initialgravityscaleoverride 0.0 _getprojectileinitialgravityscalecontext))
      ;[1 0.474 0.474] Return the projectile's gravity scale.
      (return _outxy)
      (else
        ;[1 0.474 0.474] Projectile did not set a gravity scale, return this GA's default gravity scale
        (return (ScalableFloat|GetValueatLevel (Variables|ParentDefaults|GetProjectileGravityScale)))))
    (else
      ;[1 0.474 0.474] No Projectile Class: fail safe return 0.
      (return 0.0))))

(fn GetProjectileInitialSpeedToUse
  (bind _thrownprojectileclass (Variables|ParentDefaults|GetThrownProjectileClass))
  (bind _initialspeed (Class|GetClassDefaults _thrownprojectileclass))
  (bind _getprojectileinitialspeedtousecontext (LocalVariables|GetGetProjectileInitialSpeedToUseContext))
  (if (Utilities|IsValidClass _thrownprojectileclass)
    (if (Fortnite|IsCurveTableRowValid _initialspeed _getprojectileinitialspeedtousecontext)
      (bind _outxy (Fortnite|EvaluateCurveTableRow _initialspeed 0.0 _getprojectileinitialspeedtousecontext))
      ;[1 0.474 0.474] Return the projectile's speed.
      (return _outxy)
      (else
        ;[1 0.474 0.474] Projectile did not set a speed, return this GA's default speed.
        (return (ScalableFloat|GetValueatLevel (Variables|ParentDefaults|GetProjectileSpeed)))))
    (else
      ;[1 0.474 0.474] No Projectile Class: fail safe return 0.
      (return 0.0))))

(fn GetProjectileTrajectoryActor
  (return (Variables|Trajectory|GetTrajectoryIndicatorClass)))

(fn GetProjectileTrajectoryPoints
  (bind (_outhitresult _outsplinepoints _outsplinetangents) (Trajectory|CalculateProjectileTrajectory (Default|GetOwningPlayer) (.location (LocationAndRotation|GetSpawnLocationandRotation)) (* (Math|Rotator|GetRotationXVector (.rotation (LocationAndRotation|GetSpawnLocationandRotation))) (Utilities|Struct|BreakFortProjectileTrajectoryData (Variables|Trajectory|GetProjectileTrajectoryData)))))
  (return _outsplinepoints _outsplinetangents _outhitresult))

(fn IsAbilityBlocked
  (bind _asfort_ability_system_component (Utilities|Casting|CastToFortAbilitySystemComponent (Ability|GetAbilitySystemComponentfromActorInfo))
    (:then
      (if (not (Fortnite|CanPerformNativeAction _asfort_ability_system_component (GameplayTags|MakeLiteralGameplayTag "(TagName=\"Gameplay.Action.Player.ThrowConsumable\")")))
        (return true)))
    (:CastFailed
      _))
  (return false))

(fn IsTrajectoryLocationValid
  (params (SplinePoints :type Array%20of%20Vectors) (SplineTangents :type Array%20of%20Vectors) (TrajectoryHitResult :type Hit%20Result%20Structure))
  (bind _returnvalue (Default|GetOwningPlayer))
  (bind (_bblockinghit _binitialoverlap _time _distance _location _impactpoint) (Collision|BreakHitResult TrajectoryHitResult))
  ;[0.6154 0.2705 1] Children will override this function. Called every time the trajectory is updated.
  (if (Variables|Trajectory|GetbPreventDeploymentInBlockAreas)
    (Utilities|IsValid _returnvalue
      (:Is%20Valid
        (bind _boverlapping (Default|IsOverlappingDeployableBlockVolume _returnvalue (Variables|Trajectory|GetDeployableBlockingAreaCheckRadius) (+ (Variables|Trajectory|GetDeployableBlockingAreaCheckOffset) _impactpoint) (Variables|ParentDefaults|GetUniqueBlockDeploymentTag)))
        (return (not _boverlapping)))
      (:Is%20Not%20Valid
        _))
    (else
      _))
  (return true))

(fn CanActivateAbility
  (params (ActorInfo :type Gameplay%20Ability%20Actor%20Info%20Structure) (Handle :type Gameplay%20Ability%20Spec%20Handle%20Structure))
  (bind (_owneractor _avataractor) (Utilities|Struct|BreakGameplayAbilityActorInfo ActorInfo))
  (bind _returnvalue (GameplayTags|HasAnyMatchingGameplayTags(Message) _avataractor (Variables|ParentDefaults|GetTag_BlockAbilityActivation)))
  (return 0 (not _returnvalue)))

(fn OverrideFailedReason
  (params (FailedReason :type Gameplay%20Tag%20Container%20Structure%20%28by%20ref%29))
  (bind _overridenfailedreason (|CallParentFunction FailedReason))
  (bind _asfort_pawn (Utilities|Casting|CastToFortPawn (Ability|GetAvatarActorfromActorInfo))
    (:then
      (if (GameplayTags|HasAnyMatchingGameplayTags _asfort_pawn (Variables|ParentDefaults|GetTags_UseNullFailedActivateReason))
        (return "(GameplayTags=((TagName=\"Athena.Ability.Failed.NoMessage\")))")
        (else
          _)))
    (:CastFailed
      _))
  (return _overridenfailedreason))

(fn OnUpdateTrajectory
  (params (SplinePoints :type Array%20of%20Vectors) (SplineTangents :type Array%20of%20Vectors) (TrajectoryHitResult :type Hit%20Result%20Structure)))

(fn PostThrowCleanup
  ;[1 1 1] Clear the looping trajectory display timer, destroy the trajectory indicator, and remove windup GE.
  (Variables|AbilityState|SetInThrowWindup false)
  (Ability|RemoveGameplayEffectFromOwnerWithHandle (Variables|AbilityState|GetHandle_GE_InWindup))
  (GameplayCues|RemoveCookGC)
  (if (not (Variables|AbilityState|GetSpawnedOnEquip))
    (Trajectory|CleanupTrajectory)))

(fn PrethrowSetup
  (bind (_owneractor _avataractor) (Utilities|Struct|BreakGameplayAbilityActorInfo (Ability|GetActorInfo)))
  (bind _asfort_player_pawn (Utilities|Casting|CastToFortPlayerPawn _avataractor)
    (:then
      (Variables|ReferencedActors|SetPlayerPawn _asfort_player_pawn)
      ;[1 1 1] We allow a player to cancel out of the windup for most thrown projectiles
      (Ability|SetShouldBlockOtherAbilities false)
      (Ability|SetCanbeCanceled true)
      ;[1 1 1] The input is being held down and the actor is winding up to throw the projectile
      (Variables|AbilityState|SetAbilityKeyPressed true)
      (Variables|AbilityState|SetInThrowWindup true)
      (Variables|AbilityState|SetbIsThrowing false)
      ;[1 1 1] Add a GE that applies looping GC while in the windup
      (bind _returnvalue (Ability|ApplyGameplayEffectToOwner (Variables|ParentDefaults|GetGE_InWindup)))
      (Variables|AbilityState|SetHandle_GE_InWindup _returnvalue))
    (:CastFailed
      (Ability|EndAbility))))

(fn SetupDummyProjectile
  (bind _returnvalue (Default|GetOwningPlayer))
  (bind _trajectoryindicatorclass (Variables|Trajectory|GetTrajectoryIndicatorClass))
  (if (Utilities|IsValidClass _trajectoryindicatorclass)
    (Utilities|IsValid _returnvalue
      (:Is%20Valid
        (bind _returnvalue_2 (Game|SpawnActorfromClass _trajectoryindicatorclass (Transformation|GetActorTransform _returnvalue) "AlwaysSpawn" "SelectDefaultAtRuntime" 0 _returnvalue))
        (bind _asbp_projectile_trajectory (Utilities|Casting|CastToBP_ProjectileTrajectory _returnvalue_2))
        (bind _output_get (Variables|Trajectory|SetTrajectoryIndicator _asbp_projectile_trajectory))
        (Class|BPProjectileTrajectory|SetProjectileBudgetMesh _output_get (Variables|/Script/FortniteGame.FortWeapon|GetWeaponMesh (Ability|GetCurrentSourceWeapon)))
        (TrajectoryInterface|InitTrajectoryVariables)))))

(fn ShouldOnlyShowTrajectoryOnUse
  (return true))

(fn SpawnTrajectoryIndicator
  (params (bSpawnOnEquip :type Boolean))
  (bind _returnvalue (Default|GetOwningPlayer))
  (bind _returnvalue_2 (Pawn|IsLocallyControlled _returnvalue))
  (bind _trajectoryindicator (Variables|Trajectory|GetTrajectoryIndicator))
  (Utilities|FlowControl|Sequence
    (:then_0
      (Utilities|IsValid (Variables|Trajectory|GetTrajectoryIndicator)
        (:Is%20Not%20Valid
          (Variables|AbilityState|SetSpawnedOnEquip bSpawnOnEquip)
          (Utilities|FlowControl|Sequence
            (:then_0
              ;[0.1043 1 0.1512] Attempt to create dummy projectile (i.e. trajectory indicator)
              (if _returnvalue_2
                (Trajectory|SetupDummyProjectile)))
            (:then_1
              ;[0.2834 0.3136 1] Only show trajectory display if input is held down
              (if (and _returnvalue_2 (or (Variables|AbilityState|GetAbilityKeyPressed) (Variables|AbilityState|GetSpawnedOnEquip)))
                (bind _trajectoryindicator_2 (Variables|Trajectory|GetTrajectoryIndicator)
                  (:then
                    (Class|BPProjectileTrajectory|SetShouldUpdatefromOwner _trajectoryindicator_2 true self))
                  (:else))))))))
    (:then_1
      (Utilities|IsValid _trajectoryindicator
        (:Is%20Valid
          (Variables|/Game/BP_ProjectileTrajectory.BP_ProjectileTrajectory_C|SetbWaitFirstTrajectoryValid true _trajectoryindicator))))
    (:then_2
      ;[1 0.474 0.474] Return the trajectory indciator
      (return (Variables|Trajectory|GetTrajectoryIndicator)))))

(fn UpdateTrajectorySpline
  (bind _trajectoryindicator (Variables|Trajectory|GetTrajectoryIndicator))
  (Utilities|IsValid _trajectoryindicator
    (:Is%20Valid
      (bind (_outsplinepoints _outsplinetangents _outhitresult) (TrajectoryInterface|GetProjectileTrajectoryPoints))
      (Class|BPProjectileTrajectory|PreSetTrajectorySpline _trajectoryindicator _outsplinepoints _outsplinetangents _outhitresult)
      (Trajectory|SetTrajectorySpline _trajectoryindicator _outsplinepoints _outsplinetangents)
      (Trajectory|OnUpdateTrajectory _outsplinepoints _outsplinetangents _outhitresult)
      (bind _bislocationvalid (Trajectory|IsTrajectoryLocationValid _outsplinepoints _outsplinetangents _outhitresult))
      (Class|BPProjectileTrajectory|SetTrajectoryValid _trajectoryindicator _bislocationvalid))))
