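; Interface_StickyProjectiles is an interface in the Athena consumables Parents folder for projectiles that stick to things.
; It declares a single DetachStickyProjectile function, with no logic shown, for detaching a stuck projectile.
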(fn DetachStickyProjectile)
